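# Material function for the Disco skin that computes an audio-reactive waveform, like a vibrating string, from four layered sine waves.
# Each wave scrolls over time along a coordinate gradient, with scroll, frequency and amplitude compounded from base values by per-layer curve factors.
# Audio Spectrum RGBA weights the successive waves, with alpha also scaling the fourth wave's frequency.
# The sum is divided by four and multiplied by Audio Amplitude.
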MaterialFunction /BRCosmetics/Characters/Player/Female/Medium/Bodies/F_MED_Tactical_Red_Punk/Skins/Disco/Materials/MF_AudioVisualizationString
graph {nodes [
  n1 = FunctionInput(Preview=Time()) {bUsePreviewValueAsDefault=true, inputName=Time, inputType=FunctionInput_Scalar, sortPriority=10}
  n2 = FunctionInput() {bUsePreviewValueAsDefault=true, inputName=Scroll Base, inputType=FunctionInput_Scalar, sortPriority=60, previewValue=<0.1 0 0 1>}
  n3 = FunctionInput() {bUsePreviewValueAsDefault=true, inputName=Scroll Curve Factor, inputType=FunctionInput_Scalar, sortPriority=61, previewValue=<0.5 0 0 1>}
  n4 = Multiply(A=@n2, B=@n3)
  n5 = Multiply(A=@n4, B=@n3)
  n6 = FunctionInput(Preview=MaterialFunctionCall() {materialFunction=&LinearGradient}.VGradient) {bUsePreviewValueAsDefault=true, inputName=Coordinate Gradient, inputType=FunctionInput_Scalar, sortPriority=50}
  n7 = Add(A=Multiply(A=@n1, B=@n5), B=@n6)
  n8 = FunctionInput() {bUsePreviewValueAsDefault=true, inputName=Freq Base, inputType=FunctionInput_Scalar, sortPriority=50, previewValue=<0.1 0 0 1>}
  n9 = FunctionInput() {bUsePreviewValueAsDefault=true, inputName=Freq Curve Factor, inputType=FunctionInput_Scalar, sortPriority=50, previewValue=<3 0 0 1>}
  n10 = Multiply(A=@n8, B=@n9)
  n11 = Multiply(A=@n10, B=@n9)
  n12 = FunctionInput() {bUsePreviewValueAsDefault=true, inputName=Amplitude Base, inputType=FunctionInput_Scalar, sortPriority=55, previewValue=<1 0 0 1>}
  n13 = FunctionInput() {bUsePreviewValueAsDefault=true, inputName=Amplitude Curve Factor, inputType=FunctionInput_Scalar, sortPriority=55, previewValue=<0.5 0 0 1>}
  n14 = Multiply(A=@n12, B=@n13)
  n15 = Multiply(A=@n14, B=@n13)
  n16 = Multiply(A=Sine(None=Multiply(A=@n7, B=@n11)), B=@n15)
  n17 = MaterialFunctionCall(Float4=FunctionInput() {inputName=Audio Spectrum, inputType=FunctionInput_Vector4, sortPriority=5}) {materialFunction=&BreakOutFloat4Components}
  n18 = Multiply(A=Add(A=Multiply(A=@n1, B=@n4), B=@n6), B=@n10)
  n19 = Multiply(A=Multiply(A=Sine(None=@n18), B=@n14), B=@n17.G)
  n20 = Multiply(A=Add(A=Multiply(A=@n1, B=@n2), B=@n6), B=@n8)
  n21 = Multiply(A=Multiply(A=Sine(None=@n20), B=@n12), B=@n17.R)
  n22 = Multiply(A=@n1, B=Multiply(A=@n5, B=@n3))
  n23 = Multiply(A=@n11, B=@n9)
  n24 = Sine(None=Multiply(A=Add(A=@n22, B=@n6), B=Multiply(A=@n17.A, B=@n23)))
  n25 = Multiply(A=@n15, B=@n13)
  n26 = Add(A=Add(A=Multiply(A=@n16, B=@n17.B), B=Add(A=@n19, B=@n21)), B=Multiply(A=Multiply(A=@n24, B=@n25), B=@n17.A))
  FunctionOutput(None=Multiply(A=Divide(A=@n26) {constB=4}, B=FunctionInput() {inputName=Audio Amplitude, inputType=FunctionInput_Scalar}))
 ]}
