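# Parent character material for the Tactical Red Disco skin, adding animated thin-film iridescence and audio-reactive glowing strings.
# Between shared character Pre/PostFX functions, SkinFX output receives masked, Fresnel-driven spectrum color, with a mobile switch.
# Emissive layers a curve-atlas Fresnel background, radial strings from mirrored pre-skinned position and two UV string sets, confined by mask texture channels.
# MF_AudioReactivity supplies amplitude and spectrum, which MF_AudioVisualizationString uses to distort the strings.
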Material /BRCosmetics/Characters/Player/Female/Medium/Bodies/F_MED_Tactical_Red_Punk/Skins/Disco/Materials/M_Tactical_Red_Disco_Parent
graph {
 properties {
  bAutomaticallySetUsageInEditor=false
  bHasPixelAnimation=true
  bUseMaterialAttributes=true
  bUsedWithMeshDeformer=true
  bUsedWithMorphTargets=true
  bUsedWithSkeletalMesh=true
  pixelDepthOffsetMode=PDOM_Legacy
  refractionMethod=RM_IndexOfRefraction
  shadingModel=MSM_FromMaterialExpression
  subsurfaceProfile=&SS_HeroSkin_02
  usedShadingModels=Default Lit | Subsurface | Clear Coat | Subsurface Profile | Cloth | Eye
 }
 nodes [
  n1 = MaterialFunctionCall() {materialFunction=&BaseFN_CharacterFunction_PreFX}
  n2 = MaterialFunctionCall(
   PreskinnedPosition=@n1.Pre skinned Local Position
   Gmap=@n1.Pre skinned Local Position
   MaterialAttributes=@n1.Pre skinned Local Position
   AO=@n1.Pre skinned Local Position
  ) {materialFunction=&CharacterSkinFX}
  n3 = GetMaterialAttributes(None=@n2.Result) {attributeGetTypes [#1, #2]}
  n4 = ScalarParameter() {defaultValue=0.2, group=ThinFilm, parameterName=ThinFilm_Offset}
  n5 = Sine(None=Divide(A=Time(), B=ScalarParameter() {defaultValue=4, group=ThinFilm, parameterName=ThinFilm_Animated_IdleSpeed}))
  n6 = ScalarParameter() {defaultValue=0.1429, group=ThinFilm, parameterName=ThinFilm_Animated_IdleAmplitude}
  n7 = Subtract(A=Power(Base=NamedRerouteUsage()) {constExponent=4}) {constB=0.5}
  n8 = ScalarParameter() {defaultValue=0.21, group=ThinFilm, parameterName=ThinFilm_Animated_AmplitudeScale}
  n9 = Add(A=Subtract(A=Multiply(A=@n5, B=@n6), B=Multiply(A=@n7, B=@n8)), B=@n4)
  n10 = Saturate(None=Multiply(A=NamedRerouteUsage()) {constB=100})
  n11 = ScalarParameter() {defaultValue=0.5, group=ThinFilm, parameterName=ThinFilm_Warp}
  n12 = Add(
   A=StaticSwitchParameter(True=LinearInterpolate(A=@n4, B=@n9, Alpha=@n10), False=@n4) {defaultValue=true, group=ThinFilm, parameterName=AnimateCustomThinFilm}
   B=Multiply(A=ScalarParameter() {defaultValue=1, group=ThinFilm, parameterName=ThinFilm_Scale}, B=Fresnel(ExponentIn=@n11) {baseReflectFraction=0})
  )
  n13 = Add(
   A=@n3.Base Color
   B=Multiply(
    A=TextureSampleParameter2D(UVs=@n12) {group=ThinFilm, parameterName=Thin Film Texture, texture=/Game/Global/Textures/Default/Thinfilm/T_ThinFilm_Spectrum_COLOR}.RGB
    B=ScalarParameter() {defaultValue=0.5, group=ThinFilm, parameterName=ThinFilm_Intensity}
   )
  )
  n14 = OneMinus(None=@n3.Roughness)
  n15 = ScalarParameter() {defaultValue=1, group=ThinFilm, parameterName=ThinFilm_RoughnessScale}
  n16 = Multiply(
   A=Clamp(None=Multiply(A=@n14, B=@n15))
   B=Fresnel(ExponentIn=ScalarParameter() {defaultValue=1.2164, group=ThinFilm, parameterName=ThinFilm_Exponent}) {baseReflectFraction=0}
  )
  n17 = BlendMaterialAttributes(
   A=@n2.Result
   B=SetMaterialAttributes!(MaterialAttributes=@n3.MaterialAttributes, Base Color=LinearInterpolate(A=@n3.Base Color, B=@n13, Alpha=@n16)) {attributeSetTypes [#2]}
   Alpha=ComponentMask(None=@n2.SkinFX_Mask) {r=true}
  ) {vertexAttributeBlendType=UseA}
  n18 = GetMaterialAttributes(None=StaticSwitchParameter(True=@n17, False=ShadingPathSwitch(Default=@n17, Mobile=@n2.Result)) {defaultValue=true, group=ThinFilm, parameterName=Use CustomThinFilmOnMobile}) {attributeGetTypes [#3]}
  n19 = NamedRerouteUsage()
  n20 = ScalarParameter() {defaultValue=2, group=TechArt, parameterName=ReactiveThinFilmBG_CurveExp, sortPriority=100}
  n21 = CurveAtlasRowParameter(CurveTime=OneMinus(None=Fresnel(ExponentIn=@n20) {baseReflectFraction=0})) {
   atlas=/BRCosmetics/Characters/Player/Female/Medium/Bodies/F_MED_Tactical_Red_Punk/Skins/Disco/Textures/CA_TacticalRed_Disco
   curve=/BRCosmetics/Characters/Player/Female/Medium/Bodies/F_MED_Tactical_Red_Punk/Skins/Disco/Textures/C_TacticalRed_Disco_Reactive01
   group=TechArt
   parameterName=ReactiveThinFilmBackgroundColorCurve
   sortPriority=100
  }
  n22 = Multiply(
   A=Fresnel(ExponentIn=ScalarParameter() {defaultValue=2, group=TechArt, parameterName=ReactiveThinFilmBG_OpacityExp, sortPriority=100}) {baseReflectFraction=0}
   B=@n21
  )
  n23 = ScalarParameter() {defaultValue=0.2, group=TechArt, parameterName=ReactiveThinFilmBackgroundIntensity, sortPriority=100}
  n24 = MaterialFunctionCall(Float3=PreSkinnedPosition()) {materialFunction=&BreakOutFloat3Components}
  n25 = MaterialFunctionCall(X=Abs(None=@n24.R), Y=@n24.G, Z=@n24.G) {materialFunction=&MakeFloat3}
  n26 = Multiply(
   A=Subtract(A=VertexInterpolator(VS=@n25.Result).PS, B=VectorParameter() {group=TechArt, parameterName=RadialOrigin, defaultValue=<rgba 4 -5 125 1>}.RGB)
   B=Constant3Vector() {constant=<rgba -1 1 1 0>}
  )
  n27 = ScalarParameter() {defaultValue=20, group=TechArt, parameterName=SecondaryRadius}
  n28 = MaterialFunctionCall(Vector or UVs=ComponentMask(None=Divide(A=@n26, B=@n27) {constB=18}) {b=true, r=true}) {materialFunction=&VectorToRadialValue}
  n29 = ComponentMask(None=@n28.Radial Coordinates) {r=true}
  n30 = MaterialFunctionCall!(
   Audio Amplitude=NamedRerouteUsage()
   Audio Spectrum=NamedRerouteUsage()
   Time
   Freq Base
   Coordinate Gradient=Multiply(A=@n29, B=ScalarParameter() {defaultValue=14, group=TechArt, parameterName=ExciteFrequency_Radial}) {constB=20}
   Freq Curve Factor
   Amplitude Curve Factor
   Amplitude Base
   Scroll Base
   Scroll Curve Factor
  ) {materialFunction=&/MF_AudioVisualizationString}
  n31 = Multiply(
   A=LinearInterpolate(
    A=ScalarParameter() {defaultValue=0.8, group=TechArt, parameterName=ExciteAmplitudeMin_Radial}
    B=ScalarParameter() {defaultValue=0.3, group=TechArt, parameterName=ExciteAmplitudeMax_Radial}
    Alpha=@n29
   )
   B=@n30.Result
  )
  n32 = ScalarParameter() {defaultValue=1, group=TechArt, parameterName=StringExciteIntensity_Radial}
  n33 = Subtract(
   A=Multiply(A=Time(), B=ScalarParameter() {defaultValue=0.1, group=TechArt, parameterName=PanSpeed_Radial})
   B=Add(A=ComponentMask(None=@n28.Radial Coordinates) {g=true}, B=Multiply(A=@n31, B=@n32) {constB=8})
  )
  n34 = ScalarParameter() {defaultValue=8, group=TechArt, parameterName=StringSpacing_Radial}
  n35 = SphereMask(
   A=Frac(None=Multiply(A=@n33, B=@n34) {constB=1.5})
   B=Constant() {r=0.5}
   Radius=ScalarParameter() {defaultValue=0.4, group=TechArt, parameterName=StringThickness_Radial}
  ) {attenuationRadius=0.5, hardnessPercent=0}
  n36 = ScalarParameter() {defaultValue=5, group=TechArt, parameterName=StringThicknessExp_Radial}
  n37 = LinearInterpolate(
   A=Multiply(A=ComponentMask(None=@n19) {r=true}, B=Multiply(A=@n22, B=@n23) {constB=3})
   B=LinearInterpolate(
    A=VectorParameter() {group=TechArt, parameterName=ReactivityColor_Radial, defaultValue=<rgba 0.0352 0.375 0.3127 1>}.RGB
    B=VectorParameter() {group=TechArt, parameterName=ReactivityColor_Radial_Active, defaultValue=<rgba 0 0.1414 0.1979 1>}.RGB
    Alpha=NamedRerouteUsage()
   )
   Alpha=Multiply(A=Power(Base=@n35, Exp=@n36) {constExponent=8}, B=NamedRerouteUsage())
  )
  n38 = VectorParameter() {group=TechArt, parameterName=ReactivityColor_SecondaryStrings, defaultValue=<rgba 0.392 0.0243 0.5833 1>}
  n39 = VectorParameter() {group=TechArt, parameterName=ReactivityColor_SecondaryStrings_Active, defaultValue=<rgba 1.5 0.0281 0.9415 1>}
  n40 = LinearInterpolate(A=@n38.RGB, B=@n39.RGB, Alpha=NamedRerouteUsage()) {constA=1, constB=3}
  n41 = TextureCoordinate()
  n42 = Subtract(A=Multiply(A=MaterialFunctionCall(XY=@n41) {materialFunction=&Swizzle}.YX) {constB=2}) {constB=0.98}
  n43 = ComponentMask(None=Multiply(A=Abs(None=@n42)) {constB=2}) {g=true}
  n44 = TextureCoordinate()
  n45 = MaterialFunctionCall(
   Audio Amplitude=NamedRerouteUsage()
   Audio Spectrum=NamedRerouteUsage()
   Coordinate Gradient=Multiply(
    A=ComponentMask(None=@n44) {g=true}
    B=ScalarParameter() {defaultValue=7.1983, group=TechArt, parameterName=StringGradientLength_Secondary, sortPriority=50}
   ) {constB=7.1983}
  ) {materialFunction=&MF_AudioVisualizationString}
  n46 = Multiply(A=ComponentMask(None=TextureCoordinate()) {g=true}) {constB=2.4}
  n47 = Saturate(None=OneMinus(None=@n46))
  n48 = Multiply(A=Multiply(A=@n45.Result, B=@n47), B=ScalarParameter() {defaultValue=2, group=TechArt, parameterName=StringExciteAmt})
  n49 = Add(
   A=@n46
   B=Multiply(A=Time(), B=ScalarParameter() {defaultValue=0.1896, group=TechArt, parameterName=StringBaseSineFrequency_Secondary, sortPriority=50}) {constB=0.1896}
  )
  n50 = Multiply(
   A=Multiply(A=@n47, B=Sine(None=@n49))
   B=ScalarParameter() {defaultValue=0.125, group=TechArt, parameterName=StringBaseSineAmplitude_Secondary, sortPriority=50}
  ) {constB=0.125}
  n51 = SphereMask(
   A=Add(A=Frac(None=@n43), B=Add(A=@n48, B=@n50))
   B=ScalarParameter() {defaultValue=0.42, group=TechArt, parameterName=StringOffset_Secondary, sortPriority=48}
   Radius=ScalarParameter() {defaultValue=0.22, group=TechArt, parameterName=StringThickness_Secondary, sortPriority=48}
  ) {attenuationRadius=1, hardnessPercent=0}
  n52 = ScalarParameter() {defaultValue=0.35, group=TechArt, parameterName=StringSeparation_Secondary, sliderMax=0.5, sortPriority=48}
  n53 = Add(A=@n52)
  n54 = Abs(None=OneMinus(None=Multiply(A=@n51, B=@n53)))
  n55 = Divide(A=Abs(None=Subtract(A=@n54, B=@n52) {constB=0.05}), B=OneMinus(None=@n52))
  n56 = Multiply(
   A=NamedRerouteUsage()
   B=Power(Base=OneMinus(None=@n55), Exp=ScalarParameter() {defaultValue=20, group=TechArt, parameterName=StringThicknessExp, sortPriority=38}) {constExponent=10}
  )
  n57 = VectorParameter() {group=TechArt, parameterName=ReactivityColor_PrimaryStrings, defaultValue=<rgba 0 0.4056 1 1>}
  n58 = VectorParameter() {group=TechArt, parameterName=ReactivityColor_PrimaryStrings_Active, defaultValue=<rgba 0.5176 2.669 3 1>}
  n59 = NamedRerouteUsage()
  n60 = NamedRerouteUsage()
  n61 = Multiply(A=MaterialFunctionCall(XY=TextureCoordinate()) {materialFunction=&Swizzle}.YX) {constB=2}
  n62 = Multiply(A=Abs(None=Subtract(A=@n61) {constB=0.98})) {constB=2}
  n63 = TextureCoordinate()
  n64 = MaterialFunctionCall(
   Audio Amplitude=NamedRerouteUsage()
   Audio Spectrum=NamedRerouteUsage()
   Freq Base=ScalarParameter() {defaultValue=0.1, group=TechArt, parameterName=StringFreqBase, sortPriority=40}
   Coordinate Gradient=Multiply(A=ComponentMask(None=@n63) {g=true}, B=ScalarParameter() {defaultValue=10, group=TechArt, parameterName=StringGradientLength}) {constB=6}
   Freq Curve Factor=ScalarParameter() {defaultValue=3, group=TechArt, parameterName=StringFreqCurveFactor, sortPriority=40}
   Amplitude Curve Factor=ScalarParameter() {defaultValue=0.5, group=TechArt, parameterName=StringAmpCurveFactor, sortPriority=40}
   Amplitude Base=ScalarParameter() {defaultValue=1, group=TechArt, parameterName=StringAmpBase, sortPriority=40}
   Scroll Base=ScalarParameter() {defaultValue=0.1, group=TechArt, parameterName=StringScrollBase, sortPriority=40}
   Scroll Curve Factor=ScalarParameter() {defaultValue=0.5, group=TechArt, parameterName=StringScrollCurveFactor, sortPriority=40}
  ) {materialFunction=&MF_AudioVisualizationString}
  n65 = Multiply(A=ComponentMask(None=TextureCoordinate()) {g=true}) {constB=2.4}
  n66 = Saturate(None=OneMinus(None=@n65))
  n67 = Multiply(A=@n64.Result, B=@n66)
  n68 = ScalarParameter() {defaultValue=2, group=TechArt, parameterName=StringExciteAmt}
  n69 = Multiply(A=Time(), B=ScalarParameter() {defaultValue=0.3111, group=TechArt, parameterName=StringBaseSineFrequency, sortPriority=40}) {constB=0.3111}
  n70 = Multiply(A=@n66, B=Sine(None=Add(A=@n65, B=@n69)))
  n71 = ScalarParameter() {defaultValue=0.2, group=TechArt, parameterName=StringBaseSineAmplitude, sortPriority=40}
  n72 = Add(A=Frac(None=ComponentMask(None=@n62) {g=true}), B=Add(A=Multiply(A=@n67, B=@n68), B=Multiply(A=@n70, B=@n71) {constB=0.2}))
  n73 = ScalarParameter() {defaultValue=0.4, group=TechArt, parameterName=StringOffset, sortPriority=38}
  n74 = ScalarParameter() {defaultValue=0.3, group=TechArt, parameterName=StringThickness, sortPriority=38}
  n75 = ScalarParameter() {defaultValue=0.12, group=TechArt, parameterName=StringSeparation, sliderMax=0.5, sortPriority=38}
  n76 = OneMinus(None=Multiply(A=SphereMask(A=@n72, B=@n73, Radius=@n74) {attenuationRadius=1, hardnessPercent=0}, B=Add(A=@n75)))
  n77 = Abs(None=Subtract(A=Abs(None=@n76), B=@n75) {constB=0.05})
  n78 = OneMinus(None=@n75)
  n79 = Power(
   Base=OneMinus(None=Divide(A=@n77, B=@n78))
   Exp=ScalarParameter() {defaultValue=20, group=TechArt, parameterName=StringThicknessExp, sortPriority=38}
  ) {constExponent=10}
  n80 = Multiply(
   A=NamedRerouteUsage()
   B=LinearInterpolate(
    A=LinearInterpolate(A=@n37, B=@n40, Alpha=@n56)
    B=LinearInterpolate(A=@n57.RGB, B=@n58.RGB, Alpha=@n59) {constA=1, constB=3}
    Alpha=Multiply(A=@n60, B=@n79)
   )
  )
  n81 = MaterialFunctionCall(
   AO=@n1.AO
   Distance Blend=@n1.AO
   Fresnel=@n1.AO
   Gmap=@n1.AO
   Pre Skinned Local Position=@n1.AO
   Material=SetMaterialAttributes!(MaterialAttributes=@n18.MaterialAttributes, Emissive Color=Add(A=@n18.Emissive Color, B=@n80)) {attributeSetTypes [#3]}
  ) {materialFunction=&BaseFN_CharacterFunction_PostFX}
  MaterialFunctionCall(In=@n81.Result) {materialFunction=&CINE_MF_CharPostFX}
  n82 = MaterialFunctionCall() {materialFunction=/Game/Packages/Fortress_SharedMaterials/Base_Material_Functions/BaseCharacterShaderFunctions/Audio/MF_AudioReactivity}
  n83 = ScalarParameter() {defaultValue=1, group=TechArt, parameterName=EnableAudioReactivity, sliderMax=1}
  NamedRerouteDeclaration(None=Multiply(A=@n82.Amplitude, B=@n83)) {name=Amplitude, nodeColor=<rgba 0 1 0.1094 1>}
  NamedRerouteDeclaration(None=Multiply(A=@n82.Spectrum, B=@n83)) {name=Spectrum, nodeColor=<rgba 1 0 0.75 1>}
  n84 = TextureSampleParameter2D() {
   parameterName=TechArtFX_Mask
   samplerSource=SSM_Wrap_WorldGroupSettings
   samplerType=SAMPLERTYPE_LinearColor
   texture=/BRCosmetics/Characters/Player/Female/Medium/Bodies/F_MED_Tactical_Red_Punk/Skins/Disco/Textures/T_TacticalRed_Disco_body_TechArtFX_Mask
  }
  NamedRerouteDeclaration(None=@n84.B) {name=ArmEdgeSeamMask, nodeColor=<rgba 0.6641 0 1 1>}
  NamedRerouteDeclaration(None=Power(Base=NamedRerouteUsage(), Exp=ScalarParameter() {defaultValue=8, group=TechArt, parameterName=AmplitudeEmissiveExp}) {constExponent=8}) {name=AmplitudeAdjusted, nodeColor=<rgba 0 1 0.9531 1>}
  NamedRerouteDeclaration(None=@n84.G) {name=ArmpitsMask, nodeColor=<rgba 0 0.6484 1 1>}
  NamedRerouteDeclaration(None=@n84.R) {name=ArmsMask, nodeColor=<rgba 1 0.5625 0 1>}
 ]
}
